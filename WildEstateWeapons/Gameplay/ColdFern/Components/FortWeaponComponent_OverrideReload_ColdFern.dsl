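; Weapon component for ColdFern that switches the owning weapon's override fire mode data on or off to alter its reload.
; On BeginPlay it caches the owning FortWeapon and async loads the soft-referenced fire mode asset into a variable.
; SetOverrideReload writes OverrideNextReload, whose RepNotify calls ApplyOverrideReload to set or clear the override on the cached weapon.
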(event EventBeginPlay
  (bind _override_fire_mode_data (Variables|Default|GetOverride%20Fire%20Mode%20Data))
  ;[0.09387 1 0.08132] Cache the weapon associated with this component and load the fire mode asset
  (bind _asfort_weapon (Utilities|Casting|CastToFortWeapon (Components|GetOwner)))
  (Variables|Default|SetWeapon _asfort_weapon)
  (if (Utilities|IsValidSoftObjectReference _override_fire_mode_data)
    (bind _object (Utilities|AsyncLoadAsset _override_fire_mode_data)
      (:Completed
        (Variables|Default|SetLoadedFireMode _object)))))

(fn ApplyOverrideReload
  (params (ShouldOverrideReload :type Boolean))
  (bind _weapon (Variables|Default|GetWeapon))
  (Variables|Default|GetWeapon
    (:then
      ;[0 1 1] Update the override fire mode data
      (if (Variables|GetShouldOverrideReload)
        (FortWeapon|SetOverrideFireModeData _weapon (Variables|Default|GetOverride%20Fire%20Mode%20Data))
        (else
          (FortWeapon|ClearOverrideFireModeData _weapon))))
    (:else)))

(fn OnRep_OverrideNextReload
  (CallFunction|ApplyOverrideReload (Variables|Default|GetOverrideNextReload)))

(fn SetOverrideReload
  (params (OverrideReload :type Boolean))
  (Variables|Default|SetOverrideNextReload OverrideReload))
